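# Vehicle wheel cosmetic material for StarHolo that layers a holographic, pulsing emissive effect onto the attributes from MF_WheelParent_Simple.
# A reflection-vector cubemap, tinted by Rim_ColorC, combines with Fresnel from blended rim and tire normals.
# Time-scrolled cosine rings radiate from the UV1 center, sharpened by a high power and masked by the FX texture and vertex red.
# isWhite blends between the additive glow and an inverted-ring color variant; isLOD switches the reflection masking.
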Material /VehicleCosmetics/Wheels/StarHolo/Materials/M_Wheel_StarHolo
graph {
 properties {
  bAutomaticallySetUsageInEditor=false
  bUseMaterialAttributes=true
  bUsedWithMeshDeformer=true
  bUsedWithSkeletalMesh=true
  usedShadingModels=Default Lit
 }
 nodes [
  n1 = ReflectionVectorWS()
  n2 = Desaturation(None=VectorParameter() {parameterName=Rim_ColorC, defaultValue=<rgba 0.5 1 1 0>}.RGB, Fraction=Constant()) {luminanceFactors=<rgba 0.3 0.59 0.11 0>}
  n3 = Multiply(
   A=TextureSample(
    UVs=Transform(None=@n1) {transformSourceType=TRANSFORMSOURCE_Local}
    Tex=TextureObject() {samplerType=SAMPLERTYPE_LinearColor, texture=/VehicleCosmetics/Wheels/DuelingDragonsSE/Materials/TC_DuelingDragonsSE_GrassCube}
   ) {samplerType=SAMPLERTYPE_Normal, texture=/VehicleCosmetics/Wheels/CoquetteHeart/Textures/T_Rim_CoquetteHeart_N}.RGB
   B=Desaturation(None=@n2, Fraction=Constant() {r=0.5}) {luminanceFactors=<rgba 0.3 0.59 0.11 0>}
  )
  n4 = TextureSampleParameter2D() {parameterName=Rim_Normal, samplerType=SAMPLERTYPE_Normal, texture=/VehicleCosmetics/Wheels/StarHolo/Textures/T_Rim_StarHolo_N}
  n5 = TextureSampleParameter2D() {parameterName=Tire_Normal, samplerType=SAMPLERTYPE_Normal, texture=/VehicleCosmetics/Wheels/StarHolo/Textures/T_Tire_StarHolo_N}
  n6 = MaterialFunctionCall() {materialFunction=&MF_WheelParent_Simple}
  n7 = Fresnel(
   ExponentIn=ScalarParameter() {defaultValue=1.5, group=0_SpecialEffect, parameterName=Fresnel_Exp}
   BaseReflectFractionIn=ScalarParameter() {defaultValue=0.04, group=0_SpecialEffect, parameterName=Fresnel_BaseRefl}
   Normal=Transform(None=LinearInterpolate(A=@n4.RGB, B=@n5.RGB, Alpha=@n6.VertexColorR))
  ) {exponent=1.5}
  n8 = Saturate(None=Power(Base=Multiply(A=@n7) {constB=0.75}) {constExponent=6})
  n9 = Add(A=Multiply(A=@n3, B=@n3), B=Multiply(A=@n3, B=Multiply(A=@n8, B=@n2) {constB=0.75}))
  n10 = TextureSampleParameter2D() {group=0_SpecialEffect, parameterName=_FX_LOD, texture=/VehicleCosmetics/Wheels/StarHolo/Textures/T_Rim_StarHolo_FX}
  n11 = LinearInterpolate(A=Constant(), B=@n9, Alpha=@n10.G)
  n12 = VertexColor()
  n13 = VertexColor()
  n14 = Multiply(A=Time(), B=ScalarParameter() {defaultValue=0.08, group=0_SpecialEffect, parameterName=Pulse_Speed})
  n15 = TextureCoordinate() {coordinateIndex=1}
  n16 = Distance(A=Multiply(A=Subtract(A=@n15) {constB=0.5}) {constB=2}, B=Constant2Vector())
  n17 = Multiply(A=@n14, B=ScalarParameter() {defaultValue=1, group=0_SpecialEffect, parameterName=Relative_line_speed})
  n18 = LinearInterpolate(A=Cosine(None=Add(A=@n14, B=@n16)) {period=0.075}, B=Cosine(None=Add(A=@n17, B=@n16)) {period=0.075}, Alpha=@n13.R) {constB=0}
  n19 = Power(Base=Divide(A=Add(A=@n18))) {constExponent=50}
  n20 = Multiply(
   A=LinearInterpolate(A=StaticSwitchParameter(True=@n10.G, False=@n10.G) {group=0_SpecialEffect, parameterName=isLOD}, Alpha=@n13.R)
   B=Multiply(B=Saturate(None=@n19)) {constA=3, constB=0.75}
  ) {constA=3, constB=0.75}
  n21 = Add(
   A=StaticSwitchParameter(True=LinearInterpolate(A=@n11, B=@n9, Alpha=@n12.R), False=@n9) {group=0_SpecialEffect, parameterName=isLOD}
   B=Multiply(A=@n2, B=@n20) {constB=0.75}
  ) {constB=0}
  n22 = Add(A=@n5.R, B=@n5.R)
  Power(
   Base=LinearInterpolate(A=@n9, B=@n21, Alpha=@n12.R)
   Exp=LinearInterpolate(A=@n4.G, B=ConstantBiasScale(None=@n22) {bias=1.2, scale=0.2}, Alpha=@n12.R)
  )
  n23 = Desaturation(None=VectorParameter() {parameterName=Rim_ColorC, defaultValue=<rgba 0 0 1 1>}.RGB, Fraction=Constant()) {luminanceFactors=<rgba 0.3 0.59 0.11 0>}
  n24 = Multiply(A=OneMinus(None=@n7)) {constB=0.95}
  n25 = Subtract(A=Multiply(A=@n23, B=OneMinus(None=@n20)), B=Multiply(A=Power(Base=@n24) {constExponent=3}, B=@n23))
  SetMaterialAttributes(
   MaterialAttributes=@n6.MaterialAttributes
   Emissive Color=LinearInterpolate(A=Saturate(None=@n25), B=@n21, Alpha=ScalarParameter() {defaultValue=1, group=0_SpecialEffect, parameterName=isWhite})
  ) {attributeSetTypes [#1]}
 ]
}
